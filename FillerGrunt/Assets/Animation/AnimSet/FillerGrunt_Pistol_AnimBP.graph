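# Animation Blueprint for the FillerGrunt pistol weapon skeleton that switches montage slots according to remaining ammo.
# A bool blend routes the EmptyMag slot when AmmoRemaining is zero or less, and DefaultSlot otherwise.
# Both slots sit over the same non-looping default weapon pose, so only animations played into the slots differentiate the states.
# AmmoRemaining is not declared here, and the EventGraph is empty.
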AnimBlueprint /FillerGrunt/Assets/Animation/AnimSet/FillerGrunt_Pistol_AnimBP
target_skeleton=/FillerGrunt/Assets/Meshes/FillerGrunt_Pistol_Skeleton
graphs [
 AnimGraph {nodes [
   n1 = Sequence Player 'FillerGrunt_Pistol_Pose_WEP_Default'() {bLoopAnimation=false, sequence=/FillerGrunt/Assets/Animation/Poses/FillerGrunt_Pistol_Pose_WEP_Default}
   n2 = Sequence Player 'FillerGrunt_Pistol_Pose_WEP_Default'() {bLoopAnimation=false, sequence=&FillerGrunt_Pistol_Pose_WEP_Default}
   n3 = Get AmmoRemaining() -> (AmmoRemaining:int=0)
   Output Pose(Result=Blend Poses by bool(
     BlendPose_0=Slot 'EmptyMag'(Source=@n2) {slotName=EmptyMag}
     BlendPose_1=Slot 'DefaultSlot'(Source=@n1)
     bActiveValue="integer <= integer"(A=@n3)
    ))
  ]}
 EventGraph {links [], nodes []}
]
variables []
