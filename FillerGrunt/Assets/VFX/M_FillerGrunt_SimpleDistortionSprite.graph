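# Translucent unlit sprite material for particle and Niagara effects, named as a simple distortion sprite for FillerGrunt, using index-of-refraction mode.
# A radial sphere mask, scaled by a dynamic parameter, blends a pointy normal texture against a constant to shape the distortion.
# Particle color optionally drives emissive through a shared adjustment function.
# Particle alpha times the mask forms an opacity term, with switchable depth fade.
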Material /FillerGrunt/Assets/VFX/M_FillerGrunt_SimpleDistortionSprite
graph {
 properties {
  bGenerateSphericalParticleNormals=true
  bUsedWithNiagaraSprites=true
  bUsedWithParticleSprites=true
  blendMode=BLEND_Translucent
  pixelDepthOffsetMode=PDOM_Legacy
  refractionMethod=RM_IndexOfRefraction
  shadingModel=MSM_Unlit
  usedShadingModels=Unlit
 }
 nodes [
  n1 = SphereMask(
   A=TextureCoordinate()
   B=Constant() {r=0.5}
   Radius=ScalarParameter() {defaultValue=0.5, parameterName=Radius}
   Hardness=ScalarParameter() {parameterName=Hardness}
  )
  n2 = Multiply(A=DynamicParameter().Param1, B=@n1)
  Clamp(None=LinearInterpolate(
    A=TextureSample() {samplerType=SAMPLERTYPE_Normal, texture=/Game/Effects/Fort_Effects/Textures/PointyNormal}.RGB
    B=Constant() {r=1}
    Alpha=OneMinus(None=@n2)
   ))
  n3 = ParticleColor()
  MaterialFunctionCall(Emissive=LinearInterpolate(B=@n3.RGB, Alpha=ScalarParameter() {parameterName=UseParticleColor})) {materialFunction=&MF_FX_EmissiveAdjustments}
  ScalarParameter() {defaultValue=1, parameterName=RefractionAmount}
  n4 = ScalarParameter() {parameterName=FadeDistance}
  n5 = Multiply(A=@n3.A, B=@n1)
  StaticSwitchParameter(True=Multiply(A=DepthFade(FadeDistance=@n4), B=@n5), False=@n5) {parameterName=UseDepthFade?}
 ]
}
