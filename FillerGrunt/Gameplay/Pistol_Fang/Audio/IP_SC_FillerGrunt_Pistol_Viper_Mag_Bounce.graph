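# Sound cue for the magazine bounce foley of the FillerGrunt Viper pistol.
# A Random node picks one of three bounce wave variants with equal weight.
# It plays at half volume through the Fort weapon foley sound class and attenuation, limited by the player melee weapon concurrency set.
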SoundCue /FillerGrunt/Gameplay/Pistol_Fang/Audio/IP_SC_FillerGrunt_Pistol_Viper_Mag_Bounce
root=Random_0
nodes [
 {id=Random_0, children [WavePlayer_0,WavePlayer_1,WavePlayer_2], properties {weights [1,1,1]}}
 {id=WavePlayer_0, properties {soundWaveAssetPtr=/FillerGrunt/Gameplay/Pistol_Fang/Audio/Wavs/IP_FillerGrunt_ViperPistol_Mag_Bounce_01}}
 {id=WavePlayer_1, properties {soundWaveAssetPtr=/FillerGrunt/Gameplay/Pistol_Fang/Audio/Wavs/IP_FillerGrunt_ViperPistol_Mag_Bounce_02}}
 {id=WavePlayer_2, properties {soundWaveAssetPtr=/FillerGrunt/Gameplay/Pistol_Fang/Audio/Wavs/IP_FillerGrunt_ViperPistol_Mag_Bounce_03}}
]
properties {
 attenuationSettings=&AS_Fort_WeaponFoley
 maxDistance=2250
 soundClassObject=&Fort_WeaponFoley
 volumeMultiplier=0.5
 concurrencySet [&Player_Melee_Weapon_Conc]
}
