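# Niagara system for the magazine ejected from a FillerGrunt pistol, with impact debris when it lands.
# The Magazine mesh inherits velocity, falls, tumbles and collides within LOD-distance limits, plays a bounce sound and shrinks away.
# User bools add throw velocity and switch gravity and drag for underwater.
# Magazine collisions are written to system arrays and read back, triggering limited bursts of mesh Chunks and SmokePuff sprites at the collision position.
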NiagaraSystem /FillerGrunt/Assets/VFX/NS_FillerGrunt_Magazine_Eject
dependencies {
 usedDataInterfaces [
  &NiagaraDataInterfaceAudioPlayer
  &NiagaraDataInterfaceMeshRendererInfo
  &NiagaraDataInterfacePlatformSet
  &NiagaraDataInterfaceCurve
  &NiagaraDataInterfaceArrayFloat
  &NiagaraDataInterfaceArrayInt32
  &NiagaraDataInterfaceArrayFloat2
  &NiagaraDataInterfaceColorCurve
  &NiagaraDataInterfaceArrayPosition
 ]
 usedDynamicInputs [&MaskVector2dBySpawnGroup, &RandomRangeVector2D]
 usedModules [
  &self:SystemSpawnScript.SystemScriptSource.SystemScriptGraph.NiagaraNodeAssignment_0.SetVariables_#1
  &SystemState
  &self:ReadCollision
  &self:SystemSpawnScript.SystemScriptSource.SystemScriptGraph.NiagaraNodeAssignment_1.SetVariables_#2
 ]
}
emitters {
 Chunks {
  emitterUpdateScript [
   &EmitterState(
    Enable Distance Culling=true
    MaxDistance=15000
    Module.bUseMaxDistance=true
    Module.bUseMinDistance=false
    Scalability Mode=::NewEnumerator1
   )
   /Game/Effects/Niagara/Modules/Spawn/FN_Spawn_TriggeredBurst(
    Burst Count Limit=4
    Module.Use Burst Count Limit=true
    Spawn:NiagaraBool=$System.CollisionSpawn
    Spawn Count Max=2
    Spawn Count Min=1
    Spawn Count Mode=1
    Spawn Trigger Mode=2
    Timer=0
    UseLoopCountLimit=true
   )
  ]
  particleSpawnScript [
   /SetVariables(Particles.MeshRand:NiagaraInt32=&RandomRangeInt(Minimum=0, Maximum=3)) {isSetParameters=true}
   &InitializeParticle(
    Lifetime Max=1
    Lifetime Min=0.5
    Lifetime Mode=::NewEnumerator1
    Mesh Scale=&Multiply_VectorByFloat(
     Vector=&Lerp_Vector(
      A=<0.4 0.4 0.4>
      B=<1 1 1>
      Alpha=&FloatFromCurve(FloatCurve=DataInterface() {propertyValues {Curve {keys [
           {interpMode=RCIM_Cubic, tangentMode=RCTM_User, time=0, value=0}
           {arriveTangent=1.3945, interpMode=RCIM_Cubic, leaveTangent=1.3945, tangentMode=RCTM_User, time=0.9, value=0.3}
           {arriveTangentWeight=0.0942, interpMode=RCIM_Cubic, tangentMode=RCTM_User, time=1, value=1}
          ]}}})
     )
     Float=0.75
    )
    Mesh Scale Mode=::NewEnumerator4
    Position=$System.CollisionPosition
    Position Mode=::NewEnumerator1
    UsePositionOffset=false
    Write Material Random=true
   )
   &ShapeLocation(
    MaskSpawnGroup=false
    Override Seed=false
    Rotation Mode=::NewEnumerator1
    Shape Origin=$System.CollisionPosition
    Sphere Radius=25
    UseOwnerScale=true
   )
   &AddVelocity(
    Module.Use Falloff Distance=false
    Override Seed=false
    Velocity Mode=::NewEnumerator1
    Velocity Origin=$System.CollisionPosition
    Velocity Speed=&RandomRangeFloat(Minimum=100, Maximum=200)
   )
   &InitialMeshOrientation(Use Rotation Vector=false)
  ]
  particleUpdateScript [
   &ParticleState()
   &AccelerationForce(Acceleration=&RandomRangeVector(Minimum=<0 0 -500>, Maximum=<0 0 -850>))
   &Drag(
    Drag=&FloatFromCurve(
     FloatCurve=DataInterface() {propertyValues {Curve {keys [
         {interpMode=RCIM_Cubic, leaveTangentWeight=0.5018, tangentMode=RCTM_Break, time=0, value=1}
         {arriveTangent=-4.0187, interpMode=RCIM_Cubic, tangentMode=RCTM_Break, time=0.2, value=0}
        ]}}}
     Scale Curve=20
    )
    Use Linear Drag=true
    Use Rotational Drag=false
   )
   &SolveForcesAndVelocity(Clamp Velocity=false, Limit Acceleration=false)
   &ScaleMeshSize(Scale Factor=&VectorFromFloat(Value=&FloatFromCurve(
      FloatCurve=DataInterface() {propertyValues {Curve {keys [
          <at 0 1>
          {interpMode=RCIM_Cubic, tangentMode=RCTM_User, time=0.75, value=1}
          {arriveTangent=-8.8763, interpMode=RCIM_Cubic, tangentMode=RCTM_Break, time=1, value=0.1}
         ]}}}
      Scale Curve=0.2
     )))
   &ScaleMeshSize(Scale Factor=&VectorFromFloat(Value=/Game/Effects/Niagara/DynamicInputs/Scalability/FN_ScalabilityFloatModifier(
      Value at DistanceFraction 0:NiagaraFloat=1
      Value at DistanceFraction 1:NiagaraFloat=2
      DistanceRamp:NiagaraDataInterfaceCurve=DataInterface() {propertyValues {Curve {keys [<at 0.1 0>,<at 0.5 1>]}}}
     ))) {module=ScaleMeshSize001}
   &UpdateMeshOrientation(
    Rotation Rate=&Multiply_Float(
     A=&FloatFromCurve(
      FloatCurve=DataInterface() {propertyValues {Curve {keys [
          {interpMode=RCIM_Cubic, leaveTangent=-2.1538, tangentMode=RCTM_Break, time=0, value=1}
          {interpMode=RCIM_Cubic, tangentMode=RCTM_Break, time=1, value=0}
         ]}}}
      Scale Curve=5
     )
     B=&RandomRangeFloat(Minimum=1, Maximum=1)
    )
    Rotation Vector=&RandomRangeVector(Minimum=<0 0 0>, Maximum=<1 1 1>)
   )
  ]
  rendererClasses [&NiagaraMeshRendererProperties]
  renderers [{rendererClass=&NiagaraMeshRendererProperties, rendererIndex=0}]
 }
 Magazine {
  emitterSpawnScript [&InitializeAudioPlayer(Audio Player=DataInterface() {propertyValues {
      Attenuation=null
      Concurrency=null
      MaxPlaysPerTick=10
      SoundToPlay=/FillerGrunt/Gameplay/Pistol_Fang/Audio/IP_SC_FillerGrunt_Pistol_Viper_Mag_Bounce
      bAllowLoopingOneShotSounds=false
      bLimitPlaysPerTick=true
      bOnlyActiveDuringGameplay=false
      bStopWhenComponentIsDestroyed=true
      ConfigurationUserParameter {parameter {name=None, typeDefHandle {registeredTypeIndex=1}, varData []}}
      ParameterNames []
     }})]
  emitterUpdateScript [
   &EmitterState()
   &SpawnBurst_Instantaneous(Use Spawn Probability=false, UseLoopCountLimit=false)
  ]
  particleSpawnScript [
   &self:Magazine_Initialize(Mesh Renderer Info=DataInterface() {propertyValues {MeshRenderer=&self:Magazine_0.NiagaraMeshRendererProperties_0}}) {module=Magazine_Initialize}
   &InitialMeshOrientation(Mesh Orientation Mode=::NewEnumerator1, Use Rotation Vector=false)
   &InitializeParticle(
    Color Mode=::NewEnumerator0
    Lifetime=1.25
    Mesh Scale Mode=::NewEnumerator3
    Position Offset=&RotateVectorByQuaternion(Vector To Rotate=$Particles.MeshOffset, Quaternion=$Particles.MeshOrientation)
    UsePositionOffset=true
    Write Material Random=true
   )
   &AddVelocity(
    Velocity=&TransformVector(
     Vector=&RandomRangeVector(Minimum=<0 0 -400>, Maximum=<0 0 -400>)
     Source Space=::World
     Destination Space=::Simulation
    )
    Velocity Speed Scale=0.65
   )
   &InheritVelocity(Inherited Velocity Speed Limit=0, Speed Limit=false, UseSpeedThreshold=false)
   /SetVariables(Particles.SpawnRotationVector=&RandomRangeVector(Minimum=<0.5 -1 0.5>, Maximum=<-0.5 -1 -0.5>)) {isSetParameters=true}
   &ApplyInitialForces()
   &AddVelocity(
    Velocity=&MakeCustomVectorFromBool(Bool=$User.IsUnderwater, True Vector=<3 0 0>, False Vector=<1 0 1.35>)
    Velocity Speed Scale=&MakeCustomFloatFromBool(Bool=$User.IsMagThrow, True Float=500, False Float=0)
   ) {module=AddVelocity001}
  ]
  particleUpdateScript [
   &ParticleState()
   &GravityForce(Gravity=&MakeCustomVectorFromBool(Bool=$User.IsUnderwater, True Vector=<0 0 -300>, False Vector=<0 0 -980>))
   &Collision(
    Age Colliding Particles=false
    Collision Enabled=&Boolean_AndOperation(
     Bool1=&IsPlatformSetActive(New Platform Set=DataInterface() {propertyValues {Platforms {qualityLevelMask=-1}}})
     Bool2=&Boolean_AndOperation(
      Bool1=&SetBoolByFloatComparison(A=$Engine.Owner.LODDistance, B=2500, Comparison Type=::NewEnumerator6)
      Bool2=&SetBoolByFloatComparison(A=$Particles.NormalizedAge, B=0.1)
     )
    )
    Control Roll On Collision=true
    Friction=0.4
    KillOnCollision=false
    Mesh Dimensions:Vector3f=$StackContext.MeshSize
    Module.Control Rest State Via Surface Normal=false
    Module.EnableMaxCollisionCount=false
    Module.Randomize Collision Normal Vector=true
    Particle Radius Scale=0.8
    Radius Calculation Type=::NewEnumerator1
    Radius Scale for Rotation=10
    Randomize Collision Normal=0.12
    Restitution=0.5
    ScaleRateOfRotation=true
   )
   &UpdateMeshOrientation(
    Rotation Rate=&MakeCustomFloatFromBool(Bool=$Particles.HasCollided, True Float=0, False Float=-1)
    Rotation Vector=$Particles.SpawnRotationVector
   )
   &UpdateMeshOrientation(
    Mesh Radius Calculation Method:ENiagaraCalculateRadiusOptions=::NewEnumerator0
    Mesh Renderer=DataInterface() {propertyValues {MeshRenderer=&self:Magazine_0.NiagaraMeshRendererProperties_0}}
    Orientation Method=::NewEnumerator6
   ) {module=UpdateMeshOrientation001}
   &Drag(
    Drag=&MakeCustomFloatFromBool(Bool=$User.IsUnderwater, True Float=7, False Float=1)
    Use Linear Drag=true
    Use Rotational Drag=true
   )
   &SolveForcesAndVelocity(Clamp Velocity=false, Limit Acceleration=false, Manually Enable Rotational Solver=true)
   /SetVariables(Particles.PositionWithOffset:NiagaraPosition=&SubtractVectorFromPosition(
     Position=$Particles.Position
     Vector=&RotateVectorByQuaternion(Vector To Rotate=$Particles.MeshOffset, Quaternion=$Particles.MeshOrientation)
    )) {isSetParameters=true}
   &ScaleMeshSize(Scale Factor=&VectorFromFloat(Value=&FloatFromCurve(FloatCurve=DataInterface() {propertyValues {Curve {keys [
          <at 0.02 1 RCIM_Constant>
          {interpMode=RCIM_Cubic, tangentMode=RCTM_User, time=0.9, value=1}
          {arriveTangent=-18.3792, interpMode=RCIM_Cubic, leaveTangent=-18.3792, tangentMode=RCTM_User, time=1, value=0}
         ]}}})))
   &self:WriteCollision(
    HasCollide:NiagaraBool=&InvertBool(Boolean:NiagaraBool=$Particles.HasCollided)
    Position Array:NiagaraDataInterfaceArrayPosition=$System.PositionArray
   ) {module=WriteCollision}
   /Niagara/Modules/Audio/PlayAudio(Play Audio Trigger:NiagaraBool=$Output.Collision.CollisionValid)
  ]
  rendererClasses [&NiagaraMeshRendererProperties]
  renderers [{rendererClass=&NiagaraMeshRendererProperties, rendererIndex=0}]
 }
 SmokePuff {
  emitterUpdateScript [
   &EmitterState()
   &FN_Spawn_TriggeredBurst(
    Burst Count Limit=5
    Module.Use Burst Count Limit=true
    Spawn:NiagaraBool=$System.CollisionSpawn
    Spawn Count Max=2
    Spawn Count Min=1
    Spawn Count Mode=1
    Spawn Trigger Mode=2
    Timer=0
    UseLoopCountLimit=true
   )
  ]
  particleSpawnScript [
   &InitializeParticle(
    Color=<rgba 0.6439 0.7234 0.75 0.2>
    Lifetime Max=0.7
    Lifetime Min=0.6
    Lifetime Mode=::NewEnumerator1
    Mesh Scale Mode=::NewEnumerator3
    Mesh Uniform Scale=50
    Position=$System.CollisionPosition
    Position Mode=::NewEnumerator1
    Sprite Rotation Mode=::NewEnumerator2
    Sprite Size Mode=::NewEnumerator3
    Uniform Sprite Size=&MaskFloatBySpawnGroup(
     Masked Float=&RandomRangeFloat(Minimum=10, Maximum=12)
     Passthrough Float=&Lerp_Float(A=12, B=6, Alpha=&ReturnNormalizedExecIndex())
    )
    UsePositionOffset=false
    Write Material Random=true
   ) {module=InitializeParticle002}
   &ShapeLocation(MaskSpawnGroup=false, Override Seed=false, Shape Origin=$System.CollisionPosition, Sphere Radius=0, UseOwnerScale=true)
   &AddVelocity(
    Module.Use Falloff Distance=false
    Override Seed=false
    Velocity Mode=::NewEnumerator1
    Velocity Origin=$System.CollisionPosition
    Velocity Speed=250
   )
   &CameraOffset(Camera Offset Amount=50) {module=CameraOffset001}
   &AddVelocity(Rotation Coordinate Space=::World, Velocity=<0 0 1>, Velocity Speed Scale=10) {module=AddVelocity001}
  ]
  particleUpdateScript [
   &ParticleState()
   &Drag(Drag=2.5, Use Linear Drag=true, Use Rotational Drag=true)
   &SolveForcesAndVelocity(Clamp Velocity=false, Limit Acceleration=false)
   &ScaleSpriteSize(Scale Factor=&Vector2DFromFloat(Value=&FloatFromCurve(
      FloatCurve=DataInterface() {propertyValues {Curve {keys [
          {arriveTangent=87.4549, interpMode=RCIM_Cubic, leaveTangent=87.4549, tangentMode=RCTM_User, time=0, value=1}
          {arriveTangent=4.8897, interpMode=RCIM_Cubic, leaveTangent=4.8897, tangentMode=RCTM_User, time=0.1, value=8}
          {interpMode=RCIM_Cubic, leaveTangent=-6.7553, time=1.0011, value=10}
         ]}}}
      Scale Curve=0.5
     )))
   &ScaleColor(
    Scale Alpha=&FloatFromCurve(FloatCurve=DataInterface() {propertyValues {Curve {keys [
         {arriveTangent=7.3326, interpMode=RCIM_Cubic, leaveTangent=7.3326, tangentMode=RCTM_User, time=0, value=0}
         {arriveTangent=0.0071, interpMode=RCIM_Cubic, leaveTangent=0.0071, tangentMode=RCTM_User, time=0.4, value=1}
         <at 1 0 RCIM_Cubic>
        ]}}})
    Scale RGB=&MakeVectorFromLinearColorRGB(Color=&ColorFromCurve(ColorCurve=DataInterface() {propertyValues {
        AlphaCurve {keys [<at 0 1 RCIM_Cubic>,<at 1 0 RCIM_Cubic>]}
        BlueCurve {keys [<at 0 1 RCIM_Cubic>,<at 0.3375 0.7 RCIM_Cubic>]}
        GreenCurve {keys [<at 0 1 RCIM_Cubic>,<at 0.3375 0.7 RCIM_Cubic>]}
        RedCurve {keys [<at 0 1 RCIM_Cubic>,<at 0.3375 0.7 RCIM_Cubic>]}
       }}))
    ScaleA=true
    ScaleRGB=true
   ) {module=ScaleColor001}
   &DynamicMaterialParameters(
    Index 0 Param 1=&Add_Float(
     A=&FloatFromCurve(FloatCurve=DataInterface() {propertyValues {Curve {keys [{arriveTangent=2.1549, time=0, value=0.8}, <at 1 1>]}}})
     B=&Multiply_Float(A=$Particles.Initial.MaterialRandom, B=0.5)
    )
    Index 1 Param 1=1
    Index 1 Param 2=1
    Index 1 Param 3=1
    Index 1 Param 4=1
    Index 2 Param 1=&FloatFromCurve(FloatCurve=DataInterface() {propertyValues {Curve {keys [
         {arriveTangent=1.6621, interpMode=RCIM_Cubic, leaveTangent=1.6621, tangentMode=RCTM_User, time=0, value=0.4}
         {arriveTangent=-0.0004, interpMode=RCIM_Cubic, leaveTangent=-0.0004, tangentMode=RCTM_User, time=1, value=1}
        ]}}})
    Index 2 Param 2=&RandomRangeFloat(Minimum=0.5, Maximum=1)
    Index 2 Param 3=4
    Index 2 Param 4=&RandomRangeFloat(Minimum=0.8, Maximum=1.2)
    Module.Param0WriteEnabled=true
    Module.Param0WriteEnabled1=true
    Module.Param0WriteEnabled2=true
    Module.Param0WriteEnabled3=true
    Module.Param1WriteEnabled=true
    Module.Param1WriteEnabled1=true
    Module.Param1WriteEnabled2=true
    Module.Param1WriteEnabled3=true
    Module.Param2WriteEnabled=true
    Module.Param2WriteEnabled1=true
    Module.Param2WriteEnabled2=true
    Module.Param2WriteEnabled3=true
    Module.Param3WriteEnabled=true
    Module.Param3WriteEnabled1=true
    Module.Param3WriteEnabled2=true
    Module.Param3WriteEnabled3=true
    Write Parameter Index 1=true
    Write Parameter Index 2=true
    Write Parameter Index 3=true
   ) {module=DynamicMaterialParameters002}
   &SpriteRotationRate(Rotation Rate=&RandomRangeFloat(Minimum=-90, Maximum=90))
   &SpriteRotationRate(Rotation Rate=&Multiply_Float(A=90, B=&RandomRangeFloat(Minimum=-1, Maximum=1))) {module=SpriteRotationRate001}
  ]
  rendererClasses [&NiagaraSpriteRendererProperties]
  renderers [
   {rendererClass=&NiagaraSpriteRendererProperties}
   {rendererClass=&NiagaraSpriteRendererProperties, rendererIndex=1}
  ]
 }
}
summary {
 systemName=NS_FillerGrunt_Magazine_Eject
 emitters [
  {emitterName=Magazine, rendererClasses [&NiagaraMeshRendererProperties]}
  {emitterName=Chunks, rendererClasses [&NiagaraMeshRendererProperties]}
  {emitterName=SmokePuff, rendererClasses [&NiagaraSpriteRendererProperties]}
 ]
 userVariables [
  User.IsMagThrow {type=NiagaraBool}
  User.IsUnderwater {type=NiagaraBool}
 ]
}
